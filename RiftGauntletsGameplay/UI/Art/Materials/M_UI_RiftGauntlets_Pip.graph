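# Translucent UI material that draws a Rift Gauntlets pip with a cooldown fill.
# A centre-scaled, half-turn-rotated regular polygon SDF is anti-aliased into shape, stroke and background masks.
# CooldownProgress drives a remapped vertical step for the fill, which Pip Cooldown Opacity can dim.
# The combined mask blends shared-collection Charcoal into a fill color that Gradient shifts from White to an orange vertical gradient, over a faint backing opacity.
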Material /RiftGauntletsGameplay/UI/Art/Materials/M_UI_RiftGauntlets_Pip
graph {
 properties {
  bAutomaticallySetUsageInEditor=false
  bUsedWithStaticMesh=false
  blendMode=BLEND_Translucent
  materialDomain=MD_UI
  usedShadingModels=Default Lit
 }
 nodes [
  n1 = NamedRerouteUsage()
  n2 = LinearInterpolate(Alpha=ScalarParameter() {parameterName=Pip Cooldown Opacity}) {constA=1, constB=0.6}
  n3 = Max(A=Multiply(A=Multiply(A=@n1, B=@n2), B=NamedRerouteUsage()), B=NamedRerouteUsage())
  LinearInterpolate(A=CollectionParameter() {collection=/BattleRoyaleGameUI/Shared/MPC_BRGameUI, parameterName=Charcoal}, B=NamedRerouteUsage(), Alpha=@n3)
  Max(A=@n3, B=Multiply(A=NamedRerouteUsage()) {constB=0.2})
  n4 = TextureCoordinate()
  n5 = ScalarParameter() {defaultValue=1, parameterName=Size}
  n6 = MaterialFunctionCall!(
   UVs=MaterialFunctionCall(
    UVs=MaterialFunctionCall(UVs=@n4, Texture Scale=@n5) {materialFunction=&ScaleUVsByCenter}.UVs
    "Rotation Angle (0-1)"=Constant() {r=0.5}
   ) {materialFunction=&CustomRotator}.Rotated Values
   Size
   Sides
   InnerAngle
   OutlineThickness=Constant() {r=0.05}
   OutlineAlignment=Constant() {r=-1}
   GlowMin
   GlowMax
  ) {materialFunction=/FortUILibrary/MaterialFunctions/Shapes/MF_UI_RegularPolygon}
  NamedRerouteDeclaration(None=MaterialFunctionCall(DistanceField=@n6.SDFOutline) {materialFunction=&MF_UI_AntiAliasing}.Inner) {name=PipShapeStroke, nodeColor=<rgba 0 0 0 1>}
  NamedRerouteDeclaration(None=MaterialFunctionCall(DistanceField=@n6.SDF) {materialFunction=&MF_UI_AntiAliasing}.Inner) {name=PipShape, nodeColor=<rgba 0 0 0 1>}
  NamedRerouteDeclaration(None=MaterialFunctionCall(DistanceField=@n6.SDFOutline, Threshold=Constant() {r=0.05}) {materialFunction=&MF_UI_AntiAliasing}.Inner) {name=BackgroundOpacity, nodeColor=<rgba 0 0 0 1>}
  n7 = Constant4Vector() {constant=<rgba 1 0.3684 0.0957 1>}
  n8 = Constant4Vector() {constant=<rgba 0.7835 0.2384 0.003 1>}
  n9 = TextureCoordinate()
  n10 = Power(Base=Add(A=ComponentMask(None=@n9) {g=true}) {constB=0.3}) {constExponent=5}
  NamedRerouteDeclaration(None=LinearInterpolate(
    A=CollectionParameter() {collection=&MPC_BRGameUI, parameterName=White}
    B=LinearInterpolate(A=@n7.RGBA, B=@n8.RGBA, Alpha=@n10)
    Alpha=ScalarParameter() {parameterName=Gradient}
   )) {name=FillColor, nodeColor=<rgba 0 0 0 1>}
  n11 = ComponentMask(None=TextureCoordinate()) {g=true}
  n12 = ScalarParameter() {defaultValue=1, parameterName=CooldownProgress}
  NamedRerouteDeclaration(None=Step(Y=OneMinus(None=@n11), X=LinearInterpolate(Alpha=@n12) {constA=0.305, constB=0.681})) {name=FillProgress, nodeColor=<rgba 0 0 0 1>}
 ]
}
